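# Physics asset for the M_MED_Ruckus backpack cosmetic, defining bodies and joints for its pack, two wires and three-part "butter" bone chain.
# Zero-length capsules act as spheres for the pack and butter bones, while short, thin capsules cover the wires.
# Both wires hang off the pack with moderate swing and narrow twist.
# The butter chain runs pack, upper, lower, top with tight angular limits, and its final joint is rotationally locked.
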PhysicsAsset /BRCosmetics/Accessories/FORT_Backpacks/M_MED_Ruckus/Meshes/M_MED_Ruckus_Pack_Physics
bodies {
 dyn_buttertop [{shapeType=Capsule, length=0, radius=4}]
 dyn_lowerbutter [{shapeType=Capsule, length=0, radius=6, center=<6 0 0>}]
 dyn_upperbutter [{shapeType=Capsule, length=0, radius=6, center=<6 0 0>}]
 dyn_wire_l [{shapeType=Capsule, length=5, radius=1, center=<0.8333 0 0>, rotation=<rot 90 0 0>}]
 dyn_wire_r [{shapeType=Capsule, length=5, radius=1, center=<0.8333 0 0>, rotation=<rot 90 0 0>}]
 pack [{shapeType=Capsule, length=0, radius=8, center=<0 -7 -3.5>}]
}
constraints [
 {bone1Name=dyn_wire_l, bone2Name=pack, swing1LimitDegrees=20, swing2LimitDegrees=20, twistLimitDegrees=5}
 {bone1Name=dyn_wire_r, bone2Name=pack, swing1LimitDegrees=20, swing2LimitDegrees=20, twistLimitDegrees=5}
 {bone1Name=dyn_upperbutter, bone2Name=pack, swing1LimitDegrees=10, swing2LimitDegrees=3, twistLimitDegrees=3}
 {bone1Name=dyn_lowerbutter, bone2Name=dyn_upperbutter, swing1LimitDegrees=10, swing2LimitDegrees=3, twistLimitDegrees=3}
 {
  bone1Name=dyn_buttertop
  bone2Name=dyn_lowerbutter
  swing1LimitDegrees=45
  swing1Motion=Locked
  swing2LimitDegrees=45
  swing2Motion=Locked
  twistLimitDegrees=45
  twistMotion=Locked
 }
]
