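# Control Rig for the M_MED_Ruckus backpack cosmetic skeleton, posing the small figure's two arms with IK.
# Left then right two-bone IK solves each upperarm, lowerarm and hand chain toward its hand control's transform.
# Both hand controls share a null under the dyn_butter bone, so they follow that bone.
# Each pole vector is the control's position plus a Y offset, passed through a red debug-draw node.
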ControlRigBlueprint /BRCosmetics/Accessories/FORT_Backpacks/M_MED_Ruckus/Meshes/M_MED_Ruckus_Pack_Skeleton_CtrlRig
graphs [
 RigVMFunctionLibrary {links [], nodes []}
 RigVMModel {nodes [
   RigUnit_BeginExecution = &RigUnit_BeginExecution()
   RigUnit_GetTransform = &RigUnit_GetTransform(Item=(Type=Control,Name="IK_hand l"))
   RigUnit_VisualDebugVectorItemSpace = /Script/ControlRig.RigUnit_VisualDebugVectorItemSpace(
    Color=(R=1,G=0,B=0,A=1)
    DebugDrawSettings=(DepthPriority=SDPG_Foreground,Lifetime=-1)
    Scale=5
    Space=(Type=Bone)
    Thickness=5
    bEnabled=True
    Value=/Script/RigVM.RigVMFunction_MathVectorAdd(B=(X=0,Y=20,Z=0), A=@RigUnit_GetTransform.Transform.Translation) {is=struct RigVMUnitNode, template="Add::Execute(in A,in B,out Result)"}.Result
   ) {is=struct RigVMUnitNode}
   /Script/ControlRig.RigUnit_TwoBoneIKSimplePerItem(
    DebugSettings=(Scale=10,WorldOffset=(Rotation=(X=0,Y=0,Z=0,W=1),Translation=(X=0,Y=0,Z=0),Scale3D=(X=1,Y=1,Z=1)))
    EffectorItem=(Type=Bone,Name="hand_l")
    ItemA=(Type=Bone,Name="upperarm_l")
    ItemB=(Type=Bone,Name="lowerarm_l")
    PoleVectorSpace=(Type=Bone)
    PrimaryAxis=(X=1,Y=0,Z=0)
    SecondaryAxis=(X=0,Y=1,Z=0)
    SecondaryAxisWeight=1
    StretchMaximumRatio=1.25
    StretchStartRatio=0.75
    Weight=1
    bPropagateToChildren=True
    ExecutePin=@RigUnit_BeginExecution.ExecutePin
    Effector=@RigUnit_GetTransform.Transform
    PoleVector=@RigUnit_VisualDebugVectorItemSpace.Value
   ) {is=struct RigVMUnitNode}
   n1 = &RigUnit_GetTransform(Item=(Type=Control,Name="IK hand r"))
   n2 = &RigUnit_VisualDebugVectorItemSpace(
    Color=(R=1,G=0,B=0,A=1)
    DebugDrawSettings=(DepthPriority=SDPG_Foreground,Lifetime=-1)
    Scale=5
    Space=(Type=Bone)
    Thickness=5
    bEnabled=True
    Value=&RigVMFunction_MathVectorAdd(B=(X=0,Y=20,Z=0), A=@n1.Transform.Translation) {template="Add::Execute(in A,in B,out Result)"}.Result
   )
   > &RigUnit_TwoBoneIKSimplePerItem(
    DebugSettings=(Scale=10,WorldOffset=(Rotation=(X=0,Y=0,Z=0,W=1),Translation=(X=0,Y=0,Z=0),Scale3D=(X=1,Y=1,Z=1)))
    EffectorItem=(Type=Bone,Name="hand_r")
    ItemA=(Type=Bone,Name="upperarm_r")
    ItemB=(Type=Bone,Name="lowerarm_r")
    PoleVectorSpace=(Type=Bone)
    PrimaryAxis=(X=-1,Y=0,Z=0)
    SecondaryAxis=(X=0,Y=-1,Z=0)
    SecondaryAxisWeight=1
    StretchMaximumRatio=1.25
    StretchStartRatio=0.75
    Weight=1
    bPropagateToChildren=True
    Effector=@n1.Transform
    PoleVector=@n2.Value
   )
  ]}
]
hierarchy [
 root {loc=<0 0 0>, rot=<0 0 0 1>}
 pack {parent=root, loc=<0 -0.7316 0>, rot=<0 0 0 1>}
 leg {parent=pack, loc=<0 -4.524 -23.3935>, rot=<-0.0994 -0.7001 -0.0994 -0.7001>}
 head {parent=pack, loc=<0 -4.5043 -10.5725>, rot=<0 0.7071 0 -0.7071>}
 upperarm_r {parent=head, loc=<6.9806 0.3394 -8.1517>, rot=<0.9968 0.0803 0 0>}
 lowerarm_r {parent=upperarm_r, loc=<-13.5081 0 0>, rot=<0 0 0.5335 -0.8458>}
 hand_r {parent=lowerarm_r, loc=<-12.0293 0 0>, rot=<0 0 0 -1>}
 upperarm_l {parent=head, loc=<6.9806 0.3394 8.1516>, rot=<0.0803 -0.9968 0 0>}
 lowerarm_l {parent=upperarm_l, loc=<13.5081 0 0>, rot=<0 0 0.5335 -0.8458>}
 hand_l {parent=lowerarm_l, loc=<12.0292 0 0>, rot=<0 0 0 1>}
 dyn_wire_r {parent=head, loc=<14.1035 -1.7559 -5.489>, rot=<0 0 0 -1>}
 dyn_wire_l {parent=head, loc=<14.1035 -1.7618 -2.8953>, rot=<0 0 0 -1>}
 dyn_upperbutter {parent=head, loc=<6.9806 0.3394 0>, rot=<0.0803 -0.9968 0 0>}
 dyn_lowerbutter {parent=dyn_upperbutter, loc=<13.5081 0 0>, rot=<0 0 0.5335 -0.8458>}
 dyn_butter {parent=dyn_lowerbutter, loc=<12.0293 0 0>, rot=<0 0 0 -1>}
 dyn_buttertop {parent=dyn_butter, loc=<0.3727 -7.0129 0>, rot=<-0.424 0.5658 -0.424 -0.5658>}
 IK_hand_lSpace_2 {parent=dyn_butter, type=Null, loc=<0 0 0>, rot=<0 0 0 1>}
 IK_hand_l {parent=IK_hand_lSpace_2, type=Control, loc=<-21.2671 -6.4398 8.1516>, rot=<-0.424 0.5658 -0.424 -0.5658>}
 IK_hand_r {parent=IK_hand_lSpace_2, type=Control, loc=<25.8441 22.0874 -8.1516>, rot=<0.424 -0.5658 -0.424 -0.5658>}
 lens {parent=head, loc=<7.2256 -9.6048 -0.0143>, rot=<0.5 -0.5 0.5 -0.5>}
]
