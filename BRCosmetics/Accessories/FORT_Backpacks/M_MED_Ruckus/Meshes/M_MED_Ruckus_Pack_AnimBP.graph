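# Animation Blueprint for the M_MED_Ruckus backpack cosmetic, offsetting its pack and leg bones to follow the owner mesh's spine bend.
# Each update remaps owner spine and neck yaw into bone rotations and rolls a random integer.
# When skydiving, the pose uses RigidBody physics, those rotations and a Control Rig.
# Otherwise, an idle state machine applies leg rotation, with low-weight physics in Idle and a rare random idle break.
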AnimBlueprint /BRCosmetics/Accessories/FORT_Backpacks/M_MED_Ruckus/Meshes/M_MED_Ruckus_Pack_AnimBP
target_skeleton=&/M_MED_Ruckus_Pack_Skeleton
graphs [
 AnimGraph {nodes [
   n1 = RigidBody(ComponentPose=Local To Component(LocalPose=Input Pose())) {
    bClampLinearTranslationLimitToRefPose=true
    bOverrideWorldGravity=true
    overridePhysicsAsset=&/M_MED_Ruckus_Pack_Physics
    simulationSpace=BaseBoneSpace
    alphaBoolBlend {blendOption=Linear}
    baseBoneRef {boneName=root}
    componentAppliedLinearAccClamp=<3000 3000 3000>
    componentLinearAccScale=<0.3 0.3 0.4>
    componentLinearVelScale=<0.3 0.3 0.4>
    overrideWorldGravity {z=-980}
   }
   n2 = "Transform (Modify) Bone - Bone: leg"(ComponentPose=@n1, Alpha=1, Translation="0,0,0", Rotation=Get Leg Rotate() -> (Leg Rotate:Rotator), Scale="1,1,1") {rotationMode=BMM_Additive, rotationSpace=BCS_BoneSpace, alphaBoolBlend {blendOption=Linear}, boneToModify {boneName=leg}}
   n3 = "Transform (Modify) Bone - Bone: pack"(ComponentPose=@n2, Translation="0,0,5", Rotation=Get Backpack Rotate() -> (Backpack Rotate:Rotator)) {
    rotationMode=BMM_Additive
    rotationSpace=BCS_BoneSpace
    translationMode=BMM_Additive
    alphaBoolBlend {blendOption=Linear}
    boneToModify {boneName=pack}
    translation {z=5}
   }
   n4 = "Transform (Modify) Bone - Bone: pack"(ComponentPose=@n3, Translation="0,0,0", Rotation=Get Backpack Rotate_2() -> (Backpack Rotate_2:Rotator)) {
    rotationMode=BMM_Additive
    rotationSpace=BCS_BoneSpace
    translationMode=BMM_Additive
    alphaBoolBlend {blendOption=Linear}
    boneToModify {boneName=pack}
   }
   n5 = Control Rig(Source=Component To Local(ComponentPose=@n4)) {
    alphaBoolBlend {blendOption=Linear}
    controlRigAssetReference {blueprintRigClass=&/M_MED_Ruckus_Pack_Skeleton_CtrlRig.M_MED_Ruckus_Pack_Skeleton_CtrlRig_C}
   }
   AnimGraphNode_StateMachine_0 = Idling State Machine() {is=title AnimGraphNode_StateMachine}
   Output Pose(Result=Blend Poses by bool(BlendPose_0=@n5, BlendPose_1=@AnimGraphNode_StateMachine_0, bActiveValue=Get bIsSkyDiving()) {blendType=Linear}) {layerGroup=None}
  ]}
 Idling State Machine {
  entry=Idle
  path=AnimGraph.AnimGraphNode_StateMachine_0.Idling State Machine
  ids=id
  nodes [
   AnimStateNode_0 = Idle()
   AnimStateNode_1 = Idle Break() {is=title AnimStateNode}
   AnimStateTransitionNode_0 = AnimStateTransitionNode()
   AnimStateTransitionNode_1 = AnimStateTransitionNode() {bAutomaticRuleBasedOnSequencePlayerInState=true}
  ]
  links [
   {
    dst=Idle Break
    rule=AnimGraph.AnimGraphNode_StateMachine_0.Idling State Machine.AnimStateTransitionNode_0.Transition
    src=Idle
    via=AnimStateTransitionNode_0
   }
   {
    dst=Idle
    rule=AnimGraph.AnimGraphNode_StateMachine_0.Idling State Machine.AnimStateTransitionNode_1.Transition
    src=Idle Break
    via=AnimStateTransitionNode_1
   }
  ]
 }
 Idle {
  path=AnimGraph.AnimGraphNode_StateMachine_0.Idling State Machine.AnimStateNode_0.Idle
  nodes [
   n1 = "Transform (Modify) Bone - Bone: pack"(Translation="0,0,0", Rotation=Get Backpack Rotate()) {
    rotationMode=BMM_Additive
    rotationSpace=BCS_BoneSpace
    translationMode=BMM_Additive
    alphaBoolBlend {blendOption=Linear}
    boneToModify {boneName=pack}
   }
   n2 = "Transform (Modify) Bone - Bone: pack"(ComponentPose=@n1, Translation="0,0,0", Rotation=Get Backpack Rotate_2()) {
    rotationMode=BMM_Additive
    rotationSpace=BCS_BoneSpace
    translationMode=BMM_Additive
    alphaBoolBlend {blendOption=Linear}
    boneToModify {boneName=pack}
   }
   Component To Local(ComponentPose=@n2)
   Local Space Ref Pose()
   Local To Component(LocalPose=Component To Local())
   n3 = "Transform (Modify) Bone - Bone: pack"(
    ComponentPose=Local To Component(LocalPose=Sequence Player 'M_MED_Backpack_Ruckus_Idle_BACKPACK'() {sequence=/BRCosmetics/Accessories/FORT_Backpacks/M_MED_Ruckus/Animation/M_MED_Backpack_Ruckus_Idle_BACKPACK})
    Translation="0,0,4"
    Rotation="0,0,0"
   ) {
    rotationMode=BMM_Additive
    rotationSpace=BCS_BoneSpace
    translationMode=BMM_Additive
    translationSpace=BCS_BoneSpace
    alphaBoolBlend {blendOption=Linear}
    alphaScaleBiasClamp {bInterpResult=true}
    boneToModify {boneName=pack}
    translation {z=4}
   }
   n4 = "Transform (Modify) Bone - Bone: leg"(ComponentPose=@n3, Alpha=1, Translation="0,0,0", Rotation=Get Leg Rotate(), Scale="1,1,1") {rotationMode=BMM_Additive, rotationSpace=BCS_BoneSpace, alphaBoolBlend {blendOption=Linear}, boneToModify {boneName=leg}}
   n5 = "Transform (Modify) Bone - Bone: dyn_upperbutter"(ComponentPose=@n4, Alpha=1, Translation="0,-1,0", Rotation="0,0,0", Scale="1,1,1") {
    rotationMode=BMM_Additive
    rotationSpace=BCS_BoneSpace
    translationMode=BMM_Additive
    translationSpace=BCS_BoneSpace
    alphaBoolBlend {blendOption=Linear}
    boneToModify {boneName=dyn_upperbutter}
    translation {y=-1}
   }
   n6 = "Transform (Modify) Bone - Bone: dyn_wire_l"(ComponentPose=@n5, Alpha=1, Translation="0,-1.25,0", Rotation="0,0,0", Scale="1,1,1") {
    rotationMode=BMM_Additive
    rotationSpace=BCS_BoneSpace
    translationMode=BMM_Additive
    translationSpace=BCS_BoneSpace
    alphaBoolBlend {blendOption=Linear}
    boneToModify {boneName=dyn_wire_l}
    translation {y=-1.25}
   }
   n7 = "Transform (Modify) Bone - Bone: dyn_wire_r"(ComponentPose=@n6, Alpha=1, Translation="0,-1.25,0", Rotation="0,0,0", Scale="1,1,1") {
    rotationMode=BMM_Additive
    rotationSpace=BCS_BoneSpace
    translationMode=BMM_Additive
    translationSpace=BCS_BoneSpace
    alphaBoolBlend {blendOption=Linear}
    boneToModify {boneName=dyn_wire_r}
    translation {y=-1.25}
   }
   n8 = "Transform (Modify) Bone - Bone: dyn_butter"(ComponentPose=@n7, Alpha=1, Translation="0,0,0", Rotation="0,-5,0", Scale="1,1,1") {
    rotationMode=BMM_Additive
    rotationSpace=BCS_BoneSpace
    translationMode=BMM_Additive
    translationSpace=BCS_BoneSpace
    alphaBoolBlend {blendOption=Linear}
    boneToModify {boneName=dyn_butter}
    rotation {yaw=-5}
   }
   n9 = RigidBody(ComponentPose=@n8, Alpha=0.25) {
    bClampLinearTranslationLimitToRefPose=true
    bOverrideWorldGravity=true
    overridePhysicsAsset=&M_MED_Ruckus_Pack_Physics
    simulationSpace=BaseBoneSpace
    alphaBoolBlend {blendOption=Linear}
    baseBoneRef {boneName=root}
    componentAppliedLinearAccClamp=<1000 1000 1000>
    componentLinearAccScale=<0.3 0.3 0.4>
    componentLinearVelScale=<0.3 0.3 0.4>
    overrideWorldGravity {z=-980}
   }
   n10 = Control Rig(Source=Component To Local(ComponentPose=@n9), Alpha=0) {
    alphaBoolBlend {blendOption=Linear}
    controlRigAssetReference {blueprintRigClass=&M_MED_Ruckus_Pack_Skeleton_CtrlRig_C}
   }
   Output Animation Pose(Result=@n10)
  ]
 }
 Idle Break {
  path=AnimGraph.AnimGraphNode_StateMachine_0.Idling State Machine.AnimStateNode_1.Idle Break
  nodes [
   n1 = Sequence Player 'M_MED_Backpack_Ruckus_Idle_Break_BACKPACK'() {bLoopAnimation=false, sequence=/BRCosmetics/Accessories/FORT_Backpacks/M_MED_Ruckus/Animation/M_MED_Backpack_Ruckus_Idle_Break_BACKPACK}
   n2 = "Transform (Modify) Bone - Bone: pack"(ComponentPose=Local To Component(LocalPose=@n1), Translation="0,0,4", Rotation="0,0,0") {
    rotationMode=BMM_Additive
    rotationSpace=BCS_BoneSpace
    translationMode=BMM_Additive
    translationSpace=BCS_BoneSpace
    alphaBoolBlend {blendOption=Linear}
    boneToModify {boneName=pack}
    translation {z=4}
   }
   n3 = "Transform (Modify) Bone - Bone: leg"(ComponentPose=@n2, Alpha=1, Translation="0,0,0", Rotation=Get Leg Rotate(), Scale="1,1,1") {rotationMode=BMM_Additive, rotationSpace=BCS_BoneSpace, alphaBoolBlend {blendOption=Linear}, boneToModify {boneName=leg}}
   Output Animation Pose(Result=Component To Local(ComponentPose=@n3))
  ]
 }
 Transition {
  path=AnimGraph.AnimGraphNode_StateMachine_0.Idling State Machine.AnimStateTransitionNode_0.Transition
  nodes [Result(bCanEnterTransition="integer >= integer"!(A:int=Get Random Sequence() -> (Random Sequence:int=0), B:int=499) -> (ReturnValue:bool) {is=title K2Node_CallFunction})]
 }
 Transition {path=AnimGraph.AnimGraphNode_StateMachine_0.Idling State Machine.AnimStateTransitionNode_1.Transition, nodes [Result()]}
 EventGraph {nodes [
   Event BlueprintUpdateAnimation()
   > Set Random Sequence(Random Sequence:int=RandomInteger!(Max:int=500) -> (ReturnValue:int=0)) -> (Output_Get:int=0)
   n1 = GetOwnerSkeletalMeshComponent()
   n2 = BreakRotator(InRot=BreakTransform(InTransform=GetDeltaTransformFromRefPose(self=@n1, BoneName=spine_05, BaseName=pelvis)).Rotation)
   > Set Leg Rotate(Leg Rotate:Rotator=MakeRotator(Yaw=MapRangeClamped(Value=@n2.Yaw, InRangeA=22, InRangeB=-9, OutRangeA=-16.5, OutRangeB=-35))) -> (Output_Get:Rotator)
   n3 = GetOwnerSkeletalMeshComponent()
   n4 = BreakRotator(InRot=BreakTransform(InTransform=GetDeltaTransformFromRefPose(self=@n3, BoneName=neck_01, BaseName=spine_01)).Rotation)
   > Set Backpack Rotate(Backpack Rotate:Rotator=MakeRotator(Roll=MapRangeClamped(Value=@n4.Yaw, InRangeA=-5, InRangeB=-30, OutRangeB=25))) -> (Output_Get:Rotator)
   n5 = GetOwnerSkeletalMeshComponent()
   n6 = BreakRotator(InRot=BreakTransform(InTransform=GetDeltaTransformFromRefPose(self=@n5, BoneName=neck_01, BaseName=spine_01)).Rotation)
   > Set Backpack Rotate_2(Backpack Rotate_2:Rotator=MakeRotator(Roll=MapRangeClamped(Value=@n6.Yaw, InRangeA=15, InRangeB=30, OutRangeB=15))) -> (Output_Get:Rotator)
   MapRangeClamped(Value=0, InRangeA=-5, InRangeB=-30, OutRangeB=25)
   "Scale Rotator (Float)"!(A:Rotator="0, 0, 0", B:float=-1) -> (ReturnValue:Rotator)
   TryGetPawnOwner()
  ]}
]
variables [
 Backpack Rotate {type=Rotator}
 Backpack Rotate_2 {type=Rotator}
 Leg Rotate {type=Rotator}
 Random Sequence {type=integer}
]
